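# Physics asset for the M_MED_Spy backpack cosmetic, defining secondary motion for its dangling cords, radio, antenna and two grenades.
# Each accessory pairs a base body with a parent body, all box shapes, alongside a single box for the pack.
# Angular-limited constraints link each pair, letting the cords swing most while the radio, antenna and grenades move in tighter ranges with almost no twist.
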PhysicsAsset /BRCosmetics/Accessories/FORT_Backpacks/M_MED_Spy_Backpack/Meshes/M_MED_Spy_Pack_Physics
bodies {
 dyn_cord01_base [{shapeType=Box, extentX=0.5, extentY=0.5, extentZ=0.5}]
 dyn_cord01_parent [{shapeType=Box, extentX=1.125, extentY=1.2612, extentZ=1.2719, center=<0.784 0 -0.0001>}]
 dyn_cord_base [{shapeType=Box, extentX=0.625, extentY=0.625, extentZ=0.625}]
 dyn_cord_parent [{shapeType=Box, extentX=1.125, extentY=1.2612, extentZ=1.2719, center=<0.784 0 -0.0001>}]
 dyn_grenade_01_base [{shapeType=Box, extentX=0.5, extentY=0.6875, extentZ=0.8125}]
 dyn_grenade_01_parent [{shapeType=Box, extentX=2.2308, extentY=1.0891, extentZ=3.9676, center=<1.1518 0.0609 -0.225>, rotation=<rot 0 90.0578 -84.1448>}]
 dyn_grenade_02_base [{shapeType=Box, extentX=0.5, extentY=0.6875, extentZ=0.8125}]
 dyn_grenade_02_parent [{shapeType=Box, extentX=2.2308, extentY=1.0891, extentZ=3.9676, center=<1.8878 0.3031 -0.1495>, rotation=<rot 0 90.0578 -84.1448>}]
 dyn_radio_antenna_base [{shapeType=Box, extentX=0.5, extentY=0.5, extentZ=0.5}]
 dyn_radio_antenna_parent [{shapeType=Box, extentX=0.875, extentY=0.7967, extentZ=2.0427, center=<-0.6677 0.0112 -0.0038>, rotation=<rot 0 89.0419 -89.6724>}]
 dyn_radio_base [{shapeType=Box, extentX=1, extentY=1, extentZ=1}]
 dyn_radio_parent [{shapeType=Box, extentX=9.5555, extentY=4.0155, extentZ=6.2299, center=<1.9922 0.1108 -1.602>, rotation=<rot 0 -84.5238 -89.3028>}]
 pack [{shapeType=Box, extentX=7.5, extentY=7.5, extentZ=7.5}]
}
constraints [
 {bone1Name=dyn_cord_parent, bone2Name=dyn_cord_base, swing1LimitDegrees=30, swing2LimitDegrees=30, twistLimitDegrees=1}
 {bone1Name=dyn_radio_parent, bone2Name=dyn_radio_base, swing1LimitDegrees=14, swing2LimitDegrees=6, twistLimitDegrees=2}
 {bone1Name=dyn_radio_antenna_parent, bone2Name=dyn_radio_antenna_base, swing1LimitDegrees=10, swing2LimitDegrees=10, twistLimitDegrees=1}
 {bone1Name=dyn_grenade_01_parent, bone2Name=dyn_grenade_01_base, swing1LimitDegrees=4, swing2LimitDegrees=4, twistLimitDegrees=2}
 {bone1Name=dyn_grenade_02_parent, bone2Name=dyn_grenade_02_base, swing1LimitDegrees=4, swing2LimitDegrees=4, twistLimitDegrees=2}
 {bone1Name=dyn_cord01_parent, bone2Name=dyn_cord01_base, swing1LimitDegrees=30, swing2LimitDegrees=30, twistLimitDegrees=1}
]
